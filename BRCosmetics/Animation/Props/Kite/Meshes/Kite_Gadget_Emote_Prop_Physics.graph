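# Physics asset for the kite gadget emote prop, giving the kite and its two tails simple collision and jointed physics.
# The kite body anchors two mirrored constraint chains, each running through the 02 and 06 tail bones with matching swing and twist angle limits.
# The 02 links are tiny capsules, while tail A's end capsule is much larger than tail B's.
# The root sphere has no constraint.
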PhysicsAsset /BRCosmetics/Animation/Props/Kite/Meshes/Kite_Gadget_Emote_Prop_Physics
bodies {
 kite [{shapeType=Capsule, length=47.6007, radius=30.7317, center=<0.7661 0.5547 -14.9032>, rotation=<rot 1.0064 -90 0.1575>}]
 root [{shapeType=Sphere, radius=27, shapeName=root_sphere}]
 tailA_02 [{shapeType=Capsule, length=0.505, radius=0.505, center=<-30.8424 0.0965 -0.0976>, rotation=<rot 0 -89.9976 -89.9038>}]
 tailA_06 [{shapeType=Capsule, length=112.069, radius=11, center=<8.4584 0.0994 -0.0133>, rotation=<rot 0 -89.8509 -88.9817>}]
 tailB_02 [{shapeType=Capsule, length=0.505, radius=0.505, center=<30.8643 0 0>, rotation=<rot 0 -89.9807 -89.9336>}]
 tailB_06 [{shapeType=Capsule, length=28.401, radius=2.668, center=<28.0856 0.1034 -0.7263>, rotation=<rot 0 -89.8751 -88.9784>}]
}
constraints [
 {bone1Name=tailB_02, bone2Name=kite, swing1LimitDegrees=45, swing2LimitDegrees=45, twistLimitDegrees=45}
 {bone1Name=tailB_06, bone2Name=tailB_02, swing1LimitDegrees=45, swing2LimitDegrees=45, twistLimitDegrees=45}
 {bone1Name=tailA_02, bone2Name=kite, swing1LimitDegrees=45, swing2LimitDegrees=45, twistLimitDegrees=45}
 {bone1Name=tailA_06, bone2Name=tailA_02, swing1LimitDegrees=45, swing2LimitDegrees=45, twistLimitDegrees=45}
]
